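# Master material for the Hardwood backpack that adds a timed "score" blink glow to the standard character material function stack.
# Time since the Play Score Animation parameter, divided by duration, drives a sine-sign square blink that fades out linearly.
# The blink scales base color masked by the backpack E texture's green channel.
# It adds this to CharacterSkinFX's emissive between PreFX and PostFX, then passes through cinematic PostFX.
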Material /Game/Accessories/FORT_Backpacks/Hardwood_Backpack/Materials/Med_Hardwood_Backpack_Master
graph {
 properties {
  bAutomaticallySetUsageInEditor=false
  bUseMaterialAttributes=true
  bUsedWithClothing=true
  bUsedWithMeshDeformer=true
  bUsedWithMorphTargets=true
  bUsedWithSkeletalMesh=true
  pixelDepthOffsetMode=PDOM_Legacy
  refractionMethod=RM_IndexOfRefraction
  shadingModel=MSM_FromMaterialExpression
  subsurfaceProfile=&SS_HeroSkin_02
  usedShadingModels=Default Lit | Subsurface | Clear Coat | Subsurface Profile | Cloth | Eye
 }
 nodes [
  n1 = MaterialFunctionCall() {materialFunction=&BaseFN_CharacterFunction_PreFX}
  n2 = GetMaterialAttributes(None=MaterialFunctionCall(
    PreskinnedPosition=@n1.Pre skinned Local Position
    Gmap=@n1.Pre skinned Local Position
    MaterialAttributes=@n1.Pre skinned Local Position
    AO=@n1.Pre skinned Local Position
   ) {materialFunction=&CharacterSkinFX}.Result) {attributeGetTypes [#1, #2]}
  n3 = TextureSample() {texture=/Game/Accessories/FORT_Backpacks/Hardwood_Backpack/Textures/T_Hardwood_Backpack_Backpack_E}
  n4 = Time()
  n5 = ScalarParameter() {defaultValue=-10, group=Score Animation, parameterName=Play Score Animation}
  n6 = Saturate(None=Divide(A=Subtract(A=@n4, B=@n5), B=ScalarParameter() {defaultValue=2, group=Score Animation, parameterName=Animation Duration}))
  n7 = ScalarParameter() {defaultValue=5, group=Score Animation, parameterName=Blink Speed}
  n8 = Sign(None=Sine(None=Multiply(A=@n6, B=@n7)))
  n9 = OneMinus(None=@n6)
  n10 = Multiply(
   A=Multiply(A=Multiply(A=@n2.Base Color, B=@n3.G), B=ScalarParameter() {defaultValue=3, group=Score Animation, parameterName=Blink Glow Intensity})
   B=Multiply(A=Saturate(None=@n8), B=Saturate(None=@n9))
  )
  n11 = MaterialFunctionCall(
   AO=@n1.AO
   Distance Blend=@n1.AO
   Fresnel=@n1.AO
   Gmap=@n1.AO
   Pre Skinned Local Position=@n1.AO
   Material=SetMaterialAttributes(MaterialAttributes=@n2.MaterialAttributes, Emissive Color=Add(A=@n2.Emissive Color, B=@n10)) {attributeSetTypes [#1]}
  ) {materialFunction=&BaseFN_CharacterFunction_PostFX}
  MaterialFunctionCall(In=@n11.Result) {materialFunction=&CINE_MF_CharPostFX}
 ]
}
