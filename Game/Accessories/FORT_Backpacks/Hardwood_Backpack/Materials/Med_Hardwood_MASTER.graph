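# Translucent, lit master material for the Hardwood backpack accessory, rendering a glass shell with a skin-swappable logo, defaulting to Llama.
# Fresnel on the baked normal drives glass opacity and masked emissive glow.
# Front and back offset UV ramps multiply into one coordinate, placing the logo twice, mirrored, without extra samples.
# The logo mask blends metallic, specular, roughness, flat normal, opacity and emissive overlay, before MF_TechArtOverrides_accessories applies.
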Material /Game/Accessories/FORT_Backpacks/Hardwood_Backpack/Materials/Med_Hardwood_MASTER
graph {
 properties {
  bEnableResponsiveAA=true
  bScreenSpaceReflections=true
  bUseMaterialAttributes=true
  bUsedWithClothing=true
  bUsedWithMeshDeformer=true
  bUsedWithMorphTargets=true
  bUsedWithNiagaraMeshParticles=true
  bUsedWithSkeletalMesh=true
  blendMode=BLEND_Translucent
  pixelDepthOffsetMode=PDOM_Legacy
  translucencyDirectionalLightingIntensity=5
  translucencyLightingMode=TLM_Surface
  usedShadingModels=Default Lit
 }
 nodes [
  n1 = TextureCoordinate()
  n2 = Subtract(
   A=@n1
   B=AppendVector(
    A=ScalarParameter() {defaultValue=0.744, parameterName=LogoFront_OffsetX}
    B=ScalarParameter() {defaultValue=0.55, parameterName=LogoFront_OffsetY}
   )
  )
  n3 = ScalarParameter() {defaultValue=6.3, parameterName=Logo_Scale}
  n4 = MaterialFunctionCall(Float2=Saturate(None=Multiply(A=@n2, B=@n3))) {materialFunction=&BreakOutFloat2Components}
  n5 = OneMinus(None=@n4.R)
  n6 = ScalarParameter() {defaultValue=0.505, parameterName=LogoBack_OffsetX}
  n7 = ScalarParameter() {defaultValue=0.55, parameterName=LogoBack_OffsetY}
  n8 = Multiply(A=Subtract(A=@n1, B=AppendVector(A=@n6, B=@n7)), B=ScalarParameter() {defaultValue=6.3, parameterName=Logo_Scale})
  n9 = Multiply(
   A=MaterialFunctionCall(Float2=Saturate(None=@n8)) {materialFunction=&BreakOutFloat2Components}.R
   B=@n5
  )
  LinearInterpolate(A=@n5, B=@n9)
  n10 = TextureSampleParameter2D() {
   group=Mask Texture
   parameterName=BakedNormal
   samplerSource=SSM_Wrap_WorldGroupSettings
   samplerType=SAMPLERTYPE_Normal
   sortPriority=1
   texture=/Game/Accessories/FORT_Backpacks/Hardwood_Backpack/Textures/T_Hardwood_Backpack_Backpack_N
  }
  n11 = Fresnel(ExponentIn=ScalarParameter() {defaultValue=1, parameterName=Fresnel Exponent, sortPriority=2}, Normal=Transform(None=@n10.RGB))
  n12 = Clamp(None=Multiply(A=@n11, B=ScalarParameter() {defaultValue=1, parameterName=Glass thickness, sortPriority=3}))
  n13 = AppendVector(A=Saturate(None=@n9), B=OneMinus(None=@n4.G))
  n14 = TextureSampleParameter2D(UVs=@n13) {
   group=Texture
   parameterName=MaskTexture
   samplerType=SAMPLERTYPE_Masks
   sortPriority=22
   texture=/Game/Accessories/FORT_Backpacks/Hardwood_Backpack/Skins/Llama/Textures/T_Llama_M
  }
  n15 = ChannelMaskParameter(None=@n14.RGB) {maskChannel=Green, parameterName=MaskChannel, defaultValue=<rgba 0 1 0 0>}
  n16 = LinearInterpolate(B=@n15.R, Alpha=ConstantBiasScale(None=OneMinus(None=@n11)) {bias=-0.1, scale=6})
  n17 = LinearInterpolate(
   A=LinearInterpolate(
    A=ScalarParameter() {parameterName=Fresnel Inner Transparency, sortPriority=4}
    B=ScalarParameter() {defaultValue=0.85, parameterName=Fresnel Outer Transparency, sortPriority=5}
    Alpha=@n12
   )
   B=Saturate(None=Add(A=@n16) {constB=0.2})
   Alpha=@n15.R
  )
  MaterialFunctionCall(Alpha Threshold=@n17) {materialFunction=&DitherTemporalAA_Fortnite}
  MaterialFunctionCall(Alpha Threshold=@n17) {materialFunction=&DitherTemporalAA}
  MaterialFunctionCall(UVs=@n13, Texture Scale=Constant2Vector() {g=1.2, r=1.2}) {materialFunction=&ScaleUVsByCenter}
  n18 = ScalarParameter() {defaultValue=0.1, parameterName=Metallic_Glass, sortPriority=10}
  n19 = ScalarParameter() {parameterName=Metallic_Logo, sortPriority=25}
  n20 = ScalarParameter() {defaultValue=1, parameterName=Specular_Glass, sortPriority=10}
  n21 = ScalarParameter() {parameterName=Specular_Logo, sortPriority=25}
  n22 = ScalarParameter() {parameterName=Roughness_Glass, sortPriority=10}
  n23 = ScalarParameter() {parameterName=Roughness_Logo, sortPriority=25}
  n24 = ScalarParameter() {parameterName=Fresnel Inner Emissive, sortPriority=7}
  n25 = ScalarParameter() {defaultValue=1, parameterName=Fresnel Outer Emissive, sortPriority=7}
  n26 = TextureSampleParameter2D() {
   parameterName=EmissiveMask
   sortPriority=10
   texture=/Game/Accessories/FORT_Backpacks/Hardwood_Backpack/Textures/T_Hardwood_Backpack_Backpack_E
  }
  n27 = Multiply(
   A=Multiply(
    A=LinearInterpolate(A=@n24, B=@n25, Alpha=@n12)
    B=VectorParameter() {parameterName=EmissiveColor, sortPriority=11, defaultValue=<rgba 1 1 1 0>}.RGB
   )
   B=@n26.RGB
  )
  n28 = TextureSampleParameter2D(UVs=@n13) {
   group=Diffuse Texture Overlay
   parameterName=Diffuse Texture Overlay
   sortPriority=20
   texture=/Game/Accessories/FORT_Backpacks/Hardwood_Backpack/Skins/Llama/Textures/T_Llama_D
  }
  n29 = LinearInterpolate(
   A=Constant()
   B=Multiply(A=@n28.RGB, B=ScalarParameter() {defaultValue=1, parameterName=EmissiveMult_Logo, sortPriority=26}) {constB=2}
   Alpha=Saturate(None=@n16)
  )
  n30 = Constant3Vector() {constant=<rgba 0.5 0.5 1 0>}
  n31 = SetMaterialAttributes!(
   MaterialAttributes
   Metallic=NamedRerouteDeclaration(None=LinearInterpolate(A=@n18, B=@n19, Alpha=@n15.R)) {name=metallic, nodeColor=<rgba 0.3594 1 0 1>}
   Specular=NamedRerouteDeclaration(None=LinearInterpolate(A=@n20, B=@n21, Alpha=@n15.R)) {name=specular, nodeColor=<rgba 0.2656 1 0 1>}
   Roughness=NamedRerouteDeclaration(None=LinearInterpolate(A=@n22, B=@n23, Alpha=@n15.R)) {name=roughness, nodeColor=<rgba 0.8047 0 1 1>}
   Emissive Color=NamedRerouteDeclaration(None=LinearInterpolate(A=@n27, B=@n29, Alpha=@n15.R)) {name=emissive color, nodeColor=<rgba 1 0 0.1641 1>}
   Opacity=NamedRerouteDeclaration(None=@n17) {name=opacity, nodeColor=<rgba 0.7578 1 0 1>}
   Normal=NamedRerouteDeclaration(None=LinearInterpolate(A=@n10.RGB, B=@n30, Alpha=@n15.R)) {name=normal, nodeColor=<rgba 0 1 1 1>}
  ) {attributeSetTypes [
    #1
    #2
    #3
    #4
    #5
    #6
   ]}
  MaterialFunctionCall(In=@n31) {materialFunction=&MF_TechArtOverrides_accessories}
  TextureCoordinate()
  TextureSampleParameter2D() {
   group=Texture
   parameterName=MaskTexture
   samplerType=SAMPLERTYPE_Masks
   sortPriority=22
   texture=&T_Llama_M
  }
  TextureSampleParameter2D(UVs=@n1) {
   group=Texture
   parameterName=MaskTexture
   samplerType=SAMPLERTYPE_Masks
   sortPriority=22
   texture=&T_Llama_M
  }
 ]
}
